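# Physics asset for the dynamic flower bones on the female medium Flower skeleton body cosmetic, covering flowers on both upper arms and hands.
# Each of the twelve flowers has a small base capsule and an offset, rotated capsule on the flower bone.
# One constraint joins each flower to its own base with very tight swing and twist limits, allowing only slight wobble.
# No constraints link separate flowers.
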PhysicsAsset /BRCosmetics/Characters/Player/Female/Medium/Bodies/F_MED_Flower_Skeleton/Meshes/F_MED_Flower_Skeleton_Flower_Dyn_Physics
bodies {
 dyn_hand_flowerA_base_r [{shapeType=Capsule, length=1, radius=1}]
 dyn_hand_flowerA_r [{shapeType=Capsule, length=0.75, radius=2, center=<2 0 0>, rotation=<rot 90 0 0>}]
 dyn_hand_flowerB_base_r [{shapeType=Capsule, length=1, radius=1}]
 dyn_hand_flowerB_r [{shapeType=Capsule, length=0.75, radius=2, center=<2 0 0>, rotation=<rot 90 0 0>}]
 dyn_hand_flowerC_base_r [{shapeType=Capsule, length=1, radius=1}]
 dyn_hand_flowerC_r [{shapeType=Capsule, length=0.75, radius=2, center=<2 0 0>, rotation=<rot 90 0 0>}]
 dyn_hand_flower_base_l [{shapeType=Capsule, length=1, radius=1}]
 dyn_hand_flower_l [{shapeType=Capsule, length=1.25, radius=2.25, center=<-1 0 0>, rotation=<rot 90 0 0>}]
 dyn_uparm_flowerA_base_l [{shapeType=Capsule, length=1, radius=1}]
 dyn_uparm_flowerA_base_r [{shapeType=Capsule, length=1, radius=1}]
 dyn_uparm_flowerA_l [{shapeType=Capsule, length=2, radius=1.75, center=<-2 0 0>, rotation=<rot 90 0 0>}]
 dyn_uparm_flowerA_r [{shapeType=Capsule, length=2, radius=2, center=<2 0 0>, rotation=<rot 90 0 0>}]
 dyn_uparm_flowerB_base_l [{shapeType=Capsule, length=1, radius=1}]
 dyn_uparm_flowerB_base_r [{shapeType=Capsule, length=1, radius=1}]
 dyn_uparm_flowerB_l [{shapeType=Capsule, length=2, radius=1.75, center=<-2 0 0>, rotation=<rot 90 0 0>}]
 dyn_uparm_flowerB_r [{shapeType=Capsule, length=2, radius=2, center=<2 0 0>, rotation=<rot 90 0 0>}]
 dyn_uparm_flowerC_base_l [{shapeType=Capsule, length=1, radius=1}]
 dyn_uparm_flowerC_base_r [{shapeType=Capsule, length=1, radius=1}]
 dyn_uparm_flowerC_l [{shapeType=Capsule, length=2, radius=1.75, center=<-2 0 0>, rotation=<rot 90 0 0>}]
 dyn_uparm_flowerC_r [{shapeType=Capsule, length=2, radius=2, center=<2 0 0>, rotation=<rot 90 0 0>}]
 dyn_uparm_flowerD_base_l [{shapeType=Capsule, length=1, radius=1}]
 dyn_uparm_flowerD_base_r [{shapeType=Capsule, length=1, radius=1}]
 dyn_uparm_flowerD_l [{shapeType=Capsule, length=2, radius=1.75, center=<-2 0 0>, rotation=<rot 90 0 0>}]
 dyn_uparm_flowerD_r [{shapeType=Capsule, length=2, radius=2, center=<2 0 0>, rotation=<rot 90 0 0>}]
}
constraints [
 {bone1Name=dyn_uparm_flowerA_l, bone2Name=dyn_uparm_flowerA_base_l, swing1LimitDegrees=1, swing2LimitDegrees=1, twistLimitDegrees=1}
 {bone1Name=dyn_uparm_flowerB_l, bone2Name=dyn_uparm_flowerB_base_l, swing1LimitDegrees=1, swing2LimitDegrees=1, twistLimitDegrees=1}
 {bone1Name=dyn_uparm_flowerC_l, bone2Name=dyn_uparm_flowerC_base_l, swing1LimitDegrees=1, swing2LimitDegrees=1, twistLimitDegrees=1}
 {bone1Name=dyn_uparm_flowerD_l, bone2Name=dyn_uparm_flowerD_base_l, swing1LimitDegrees=3, swing2LimitDegrees=3, twistLimitDegrees=3}
 {bone1Name=dyn_uparm_flowerA_r, bone2Name=dyn_uparm_flowerA_base_r, swing1LimitDegrees=1, swing2LimitDegrees=1, twistLimitDegrees=1}
 {bone1Name=dyn_uparm_flowerB_r, bone2Name=dyn_uparm_flowerB_base_r, swing1LimitDegrees=1, swing2LimitDegrees=1, twistLimitDegrees=1}
 {bone1Name=dyn_uparm_flowerC_r, bone2Name=dyn_uparm_flowerC_base_r, swing1LimitDegrees=1, swing2LimitDegrees=1, twistLimitDegrees=1}
 {bone1Name=dyn_uparm_flowerD_r, bone2Name=dyn_uparm_flowerD_base_r, swing1LimitDegrees=3, swing2LimitDegrees=3, twistLimitDegrees=3}
 {bone1Name=dyn_hand_flowerC_r, bone2Name=dyn_hand_flowerC_base_r, swing1LimitDegrees=2, swing2LimitDegrees=2, twistLimitDegrees=2}
 {bone1Name=dyn_hand_flowerB_r, bone2Name=dyn_hand_flowerB_base_r, swing1LimitDegrees=2, swing2LimitDegrees=2, twistLimitDegrees=2}
 {bone1Name=dyn_hand_flowerA_r, bone2Name=dyn_hand_flowerA_base_r, swing1LimitDegrees=2, swing2LimitDegrees=2, twistLimitDegrees=2}
 {bone1Name=dyn_hand_flower_l, bone2Name=dyn_hand_flower_base_l, swing1LimitDegrees=1, swing2LimitDegrees=1, twistLimitDegrees=1}
]
