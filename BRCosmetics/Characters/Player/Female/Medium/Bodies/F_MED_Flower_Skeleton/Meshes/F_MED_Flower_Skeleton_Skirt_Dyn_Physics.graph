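# Physics asset for the skirt of the F_MED_Flower_Skeleton female cosmetic body, modeling pelvis, thighs and four skirt panels as capsules.
# Thighs attach to the pelvis with free angular joints, and front and back panels hang from them with restricted swing and twist.
# The rig is asymmetric: right panels attach directly to the thigh, while left panels go through small base bodies with free joints to the thigh.
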PhysicsAsset /BRCosmetics/Characters/Player/Female/Medium/Bodies/F_MED_Flower_Skeleton/Meshes/F_MED_Flower_Skeleton_Skirt_Dyn_Physics
bodies {
 dyn_skirt_back_base_l [{shapeType=Capsule, length=1, radius=1}]
 dyn_skirt_back_l [{shapeType=Capsule, length=9.25, radius=7.25, shapeName=dyn_skirt_back_l_capsule, center=<5 -3.7 0>}]
 dyn_skirt_back_r [{shapeType=Capsule, length=9.25, radius=7.25, shapeName=dyn_skirt_back_r_capsule, center=<-5 4.2 0>}]
 dyn_skirt_front_base_l [{shapeType=Capsule, length=1, radius=1}]
 dyn_skirt_front_l [{shapeType=Capsule, length=9.25, radius=7.25, shapeName=dyn_skirt_front_l_capsule, center=<5 -6 0>}]
 dyn_skirt_front_r [{shapeType=Capsule, length=9.25, radius=7.25, shapeName=dyn_skirt_front_r_capsule, center=<-7 6.2 0>}]
 pelvis [{shapeType=Capsule, length=4.5, radius=4.25}]
 thigh_l [{shapeType=Capsule, length=13, radius=7.5, center=<-13.3915 0 -0.0002>, rotation=<rot 90 -0.0099 -0.0099>}]
 thigh_r [{shapeType=Capsule, length=12.75, radius=7.75, center=<13.3915 0 0.0002>, rotation=<rot 90 0.0099 -179.9901>}]
}
constraints [
 {
  bone1Name=thigh_l
  bone2Name=pelvis
  swing1LimitDegrees=45
  swing1Motion=Free
  swing2LimitDegrees=45
  swing2Motion=Free
  twistLimitDegrees=45
  twistMotion=Free
 }
 {
  bone1Name=dyn_skirt_back_base_l
  bone2Name=thigh_l
  swing1LimitDegrees=45
  swing1Motion=Free
  swing2LimitDegrees=45
  swing2Motion=Free
  twistLimitDegrees=45
  twistMotion=Free
 }
 {
  bone1Name=dyn_skirt_front_base_l
  bone2Name=thigh_l
  swing1LimitDegrees=45
  swing1Motion=Free
  swing2LimitDegrees=45
  swing2Motion=Free
  twistLimitDegrees=45
  twistMotion=Free
 }
 {bone1Name=dyn_skirt_front_l, bone2Name=dyn_skirt_front_base_l, swing1LimitDegrees=25, swing2LimitDegrees=5, twistLimitDegrees=5}
 {
  bone1Name=dyn_skirt_back_l
  bone2Name=dyn_skirt_back_base_l
  swing1LimitDegrees=45
  swing2LimitDegrees=5
  swing2Motion=Locked
  twistLimitDegrees=5
  twistMotion=Locked
 }
 {bone1Name=dyn_skirt_front_r, bone2Name=thigh_r, swing1LimitDegrees=40, swing2LimitDegrees=30, twistLimitDegrees=10}
 {bone1Name=dyn_skirt_back_r, bone2Name=thigh_r, swing1LimitDegrees=20, swing2LimitDegrees=5, twistLimitDegrees=5}
 {
  bone1Name=thigh_r
  bone2Name=pelvis
  swing1LimitDegrees=45
  swing1Motion=Free
  swing2LimitDegrees=45
  swing2Motion=Free
  twistLimitDegrees=45
  twistMotion=Free
 }
]
