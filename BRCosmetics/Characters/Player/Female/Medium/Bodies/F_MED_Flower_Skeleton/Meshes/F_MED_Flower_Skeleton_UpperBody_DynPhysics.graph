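# Physics asset for the Flower Skeleton female medium upper body, setting up dynamic motion for upper-arm cloth bones on both arms.
# Each arm has two capsule chains (a, b) with limited swing and twist, plus a larger upper-arm twist capsule.
# The left a chain forks into two extra tips.
# Left bases and twist link to upperarm_l with free angular motion, while right box bases and twist have no upper-arm constraint.
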PhysicsAsset /BRCosmetics/Characters/Player/Female/Medium/Bodies/F_MED_Flower_Skeleton/Meshes/F_MED_Flower_Skeleton_UpperBody_DynPhysics
bodies {
 dyn_uparm_cloth_a_base_l [{shapeType=Capsule, length=1, radius=1}]
 dyn_uparm_cloth_a_base_r [{shapeType=Box, extentX=1, extentY=1, extentZ=1}]
 dyn_uparm_cloth_a_l [{shapeType=Capsule, length=3, radius=2, shapeName=dyn_uparm_cloth_a_l_capsule, center=<3.5 0.5 0>}]
 dyn_uparm_cloth_a_r [{shapeType=Capsule, length=1.875, radius=3.125, shapeName=dyn_uparm_cloth_a_r_capsule, center=<-3 -2 0>}]
 dyn_uparm_cloth_a_tip_a_l [{shapeType=Capsule, length=1.875, radius=2, shapeName=dyn_uparm_cloth_a_tip_a_l_capsule, center=<2 0.5 0>}]
 dyn_uparm_cloth_a_tip_b_l [{shapeType=Capsule, length=2.25, radius=2, shapeName=dyn_uparm_cloth_a_tip_b_l_capsule, center=<2 0.5 0>}]
 dyn_uparm_cloth_a_tip_l [{shapeType=Capsule, length=3, radius=2, shapeName=dyn_uparm_cloth_a_tip_l_capsule, center=<3 0.5 0>}]
 dyn_uparm_cloth_a_tip_r [{shapeType=Capsule, length=1.875, radius=3.125, shapeName=dyn_uparm_cloth_a_tip_r_capsule, center=<-3 -2 0>}]
 dyn_uparm_cloth_b_base_l [{shapeType=Capsule, length=1, radius=1}]
 dyn_uparm_cloth_b_base_r [{shapeType=Box, extentX=1, extentY=1, extentZ=1}]
 dyn_uparm_cloth_b_l [{shapeType=Capsule, length=2.375, radius=2.625, shapeName=dyn_uparm_cloth_b_l_capsule, center=<4 -0.8 0>}]
 dyn_uparm_cloth_b_r [{shapeType=Capsule, length=1.75, radius=3.25, shapeName=dyn_uparm_cloth_b_r_capsule, center=<-3 1 0>}]
 dyn_uparm_cloth_b_tip_l [{
   shapeType=Capsule
   length=2.375
   radius=2.625
   shapeName=dyn_uparm_cloth_b_tip_l_capsule
   center=<4 -0.8 0>
   rotation=<rot 30 0.0002 -0.0002>
  }]
 dyn_uparm_cloth_b_tip_r [{shapeType=Capsule, length=1.75, radius=3.25, shapeName=dyn_uparm_cloth_b_tip_r_capsule, center=<-4 1 0>}]
 upperarm_l [{shapeType=Capsule, length=1, radius=1}]
 upperarm_twist_01_l [{shapeType=Capsule, length=16.8097, radius=4.1595, center=<6.0382 -0.4124 -0.7068>, rotation=<rot 0 -92.1235 -90.4115>}]
 upperarm_twist_01_r [{shapeType=Capsule, length=16.8097, radius=4.1595, center=<-6.0382 0.4124 0.7068>, rotation=<rot 0 -92.1235 89.5885>}]
}
constraints [
 {
  bone1Name=upperarm_twist_01_l
  bone2Name=upperarm_l
  swing1LimitDegrees=45
  swing1Motion=Free
  swing2LimitDegrees=45
  swing2Motion=Free
  twistLimitDegrees=45
  twistMotion=Free
 }
 {bone1Name=dyn_uparm_cloth_a_l, bone2Name=dyn_uparm_cloth_a_base_l, swing1LimitDegrees=20, swing2LimitDegrees=20, twistLimitDegrees=15}
 {bone1Name=dyn_uparm_cloth_a_tip_l, bone2Name=dyn_uparm_cloth_a_l, swing1LimitDegrees=20, swing2LimitDegrees=10, twistLimitDegrees=15}
 {
  bone1Name=dyn_uparm_cloth_a_base_l
  bone2Name=upperarm_l
  swing1LimitDegrees=45
  swing1Motion=Free
  swing2LimitDegrees=45
  swing2Motion=Free
  twistLimitDegrees=45
  twistMotion=Free
 }
 {
  bone1Name=dyn_uparm_cloth_b_base_l
  bone2Name=upperarm_l
  swing1LimitDegrees=45
  swing1Motion=Free
  swing2LimitDegrees=45
  swing2Motion=Free
  twistLimitDegrees=45
  twistMotion=Free
 }
 {bone1Name=dyn_uparm_cloth_b_l, bone2Name=dyn_uparm_cloth_b_base_l, swing1LimitDegrees=15, swing2LimitDegrees=15, twistLimitDegrees=5}
 {bone1Name=dyn_uparm_cloth_b_tip_l, bone2Name=dyn_uparm_cloth_b_l, swing1LimitDegrees=30, swing2LimitDegrees=45, twistLimitDegrees=20}
 {bone1Name=dyn_uparm_cloth_a_tip_b_l, bone2Name=dyn_uparm_cloth_a_tip_l, swing1LimitDegrees=20, swing2LimitDegrees=5, twistLimitDegrees=15}
 {bone1Name=dyn_uparm_cloth_a_tip_a_l, bone2Name=dyn_uparm_cloth_a_tip_l, swing1LimitDegrees=20, swing2LimitDegrees=5, twistLimitDegrees=15}
 {bone1Name=dyn_uparm_cloth_b_r, bone2Name=dyn_uparm_cloth_b_base_r, swing1LimitDegrees=15, swing2LimitDegrees=20, twistLimitDegrees=15}
 {bone1Name=dyn_uparm_cloth_b_tip_r, bone2Name=dyn_uparm_cloth_b_r, swing1LimitDegrees=20, swing2LimitDegrees=30, twistLimitDegrees=15}
 {bone1Name=dyn_uparm_cloth_a_r, bone2Name=dyn_uparm_cloth_a_base_r, swing1LimitDegrees=15, swing2LimitDegrees=15, twistLimitDegrees=15}
 {bone1Name=dyn_uparm_cloth_a_tip_r, bone2Name=dyn_uparm_cloth_a_r, swing1LimitDegrees=30, swing2LimitDegrees=30, twistLimitDegrees=15}
]
