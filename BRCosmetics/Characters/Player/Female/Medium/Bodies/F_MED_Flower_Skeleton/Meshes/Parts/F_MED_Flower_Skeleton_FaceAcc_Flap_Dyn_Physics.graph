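# Physics asset for the back flaps on the female medium Flower Skeleton cosmetic's face accessory part.
# Two capsule chains, A and B, each run from a small base body through three segments, with swing mostly allowed on one axis and tight limits on the other axis and twist.
# Both bases link to spine_04 with free angular motion, as do the B-to-A middle segments.
# Larger spine capsules provide body volumes alongside the flaps.
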PhysicsAsset /BRCosmetics/Characters/Player/Female/Medium/Bodies/F_MED_Flower_Skeleton/Meshes/Parts/F_MED_Flower_Skeleton_FaceAcc_Flap_Dyn_Physics
bodies {
 dyn_backflapA_a [{shapeType=Capsule, length=17.75, radius=3.5, shapeName=dyn_backflapA_a_capsule, center=<2 -2 0>}]
 dyn_backflapA_b [{shapeType=Capsule, length=14.125, radius=5.125, shapeName=dyn_backflapA_b_capsule, center=<4 -2 0>}]
 dyn_backflapA_base [{shapeType=Capsule, length=1, radius=1}]
 dyn_backflapA_c [{shapeType=Capsule, length=7.625, radius=5.75, shapeName=dyn_backflapA_c_capsule, center=<5 -2 0>}]
 dyn_backflapB_a [{shapeType=Capsule, length=9, radius=3.75, shapeName=dyn_backflapB_a_capsule, center=<2 -4 0>}]
 dyn_backflapB_b [{shapeType=Capsule, length=8.5, radius=3.625, shapeName=dyn_backflapB_b_capsule, center=<3 -4 0>}]
 dyn_backflapB_base [{shapeType=Capsule, length=1, radius=1}]
 dyn_backflapB_c [{shapeType=Capsule, length=11.25, radius=3.75, shapeName=dyn_backflapB_c_capsule, center=<4 -4 0>}]
 spine_01 [{shapeType=Capsule, length=21.5, radius=8.5, shapeName=spine_01_capsule, center=<1.2011 0.7376 0>}]
 spine_03 [{shapeType=Capsule, length=13.875, radius=7.375, shapeName=spine_03_capsule, center=<-0.211 0.6748 0>}]
 spine_04 [{shapeType=Capsule, length=17.5, radius=6.875, shapeName=spine_05_capsule, center=<0 0.6473 0>}]
 spine_05 [{shapeType=Capsule, length=16.25, radius=8.125, shapeName=spine_05_capsule, center=<0 1.0402 0>}]
}
constraints [
 {bone1Name=dyn_backflapA_a, bone2Name=dyn_backflapA_base, swing1LimitDegrees=40, swing2LimitDegrees=5, twistLimitDegrees=5}
 {bone1Name=dyn_backflapA_b, bone2Name=dyn_backflapA_a, swing1LimitDegrees=40, swing2LimitDegrees=5, twistLimitDegrees=5}
 {bone1Name=dyn_backflapB_b, bone2Name=dyn_backflapB_a, swing1LimitDegrees=15, swing2LimitDegrees=5, twistLimitDegrees=5}
 {bone1Name=dyn_backflapB_a, bone2Name=dyn_backflapB_base, swing1LimitDegrees=10, swing2LimitDegrees=5, twistLimitDegrees=5}
 {
  bone1Name=dyn_backflapB_base
  bone2Name=spine_04
  swing1LimitDegrees=45
  swing1Motion=Free
  swing2LimitDegrees=45
  swing2Motion=Free
  twistLimitDegrees=45
  twistMotion=Free
 }
 {bone1Name=dyn_backflapA_c, bone2Name=dyn_backflapA_b, swing1LimitDegrees=30, swing2LimitDegrees=5, twistLimitDegrees=5}
 {
  bone1Name=spine_05
  bone2Name=spine_04
  swing1LimitDegrees=45
  swing1Motion=Free
  swing2LimitDegrees=45
  swing2Motion=Free
  twistLimitDegrees=45
  twistMotion=Free
 }
 {bone1Name=dyn_backflapB_c, bone2Name=dyn_backflapB_b, swing1LimitDegrees=10, swing2LimitDegrees=5, twistLimitDegrees=5}
 {
  bone1Name=dyn_backflapB_b
  bone2Name=dyn_backflapA_b
  swing1LimitDegrees=45
  swing1Motion=Free
  swing2LimitDegrees=45
  swing2Motion=Free
  twistLimitDegrees=45
  twistMotion=Free
 }
 {
  bone1Name=dyn_backflapA_base
  bone2Name=spine_04
  swing1LimitDegrees=45
  swing1Motion=Free
  swing2LimitDegrees=45
  swing2Motion=Free
  twistLimitDegrees=45
  twistMotion=Free
 }
 {
  bone1Name=spine_04
  bone2Name=spine_03
  swing1LimitDegrees=45
  swing1Motion=Free
  swing2LimitDegrees=45
  swing2Motion=Free
  twistLimitDegrees=45
  twistMotion=Free
 }
]
