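# Physics asset for the face accessory mesh of the female medium Flower Skeleton player cosmetic, defining physics for the flowers attached to the head.
# Six dyn_head_flower bones each carry a capsule body, and a small capsule on the head bone serves as their shared parent.
# Each flower is constrained directly to the head rather than chained, with limited swing and a tighter twist limit, allowing slight independent wobble.
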PhysicsAsset /BRCosmetics/Characters/Player/Female/Medium/Bodies/F_MED_Flower_Skeleton/Meshes/Parts/F_MED_Flower_Skeleton_FaceAcc_Physics
bodies {
 dyn_head_flower_a [{shapeType=Capsule, length=1.75, radius=2.5, center=<-1.8237 0.0001 -0.0001>, rotation=<rot 90 -0.0396 -0.0396>}]
 dyn_head_flower_b [{shapeType=Capsule, length=3, radius=2.25, center=<-1.1199 0 -0.0001>, rotation=<rot 90 0 0>}]
 dyn_head_flower_c [{shapeType=Capsule, length=3, radius=2.25, center=<-1.1199 0 -0.0001>, rotation=<rot 90 0 0>}]
 dyn_head_flower_d [{shapeType=Capsule, length=3, radius=2.25, center=<-1.1199 0 -0.0001>, rotation=<rot 90 0 0>}]
 dyn_head_flower_e [{shapeType=Capsule, length=3, radius=2.25, center=<-1.1199 0 -0.0001>, rotation=<rot 90 0 0>}]
 dyn_head_flower_f [{shapeType=Capsule, length=3, radius=2.25, center=<-1.1199 0 -0.0001>, rotation=<rot 90 0 0>}]
 head [{shapeType=Capsule, length=1, radius=1}]
}
constraints [
 {bone1Name=dyn_head_flower_a, bone2Name=head, swing1LimitDegrees=15, swing2LimitDegrees=15, twistLimitDegrees=3}
 {bone1Name=dyn_head_flower_b, bone2Name=head, swing1LimitDegrees=15, swing2LimitDegrees=15, twistLimitDegrees=3}
 {bone1Name=dyn_head_flower_c, bone2Name=head, swing1LimitDegrees=15, swing2LimitDegrees=15, twistLimitDegrees=3}
 {bone1Name=dyn_head_flower_d, bone2Name=head, swing1LimitDegrees=15, swing2LimitDegrees=15, twistLimitDegrees=3}
 {bone1Name=dyn_head_flower_e, bone2Name=head, swing1LimitDegrees=15, swing2LimitDegrees=15, twistLimitDegrees=3}
 {bone1Name=dyn_head_flower_f, bone2Name=head, swing1LimitDegrees=15, swing2LimitDegrees=15, twistLimitDegrees=3}
]
