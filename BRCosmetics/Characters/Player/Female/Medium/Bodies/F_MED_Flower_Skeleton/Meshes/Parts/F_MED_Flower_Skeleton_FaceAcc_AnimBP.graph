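# Post-process style AnimBlueprint for the Flower Skeleton skin's FaceAcc part, adding secondary motion to vine, flowers and back flaps.
# After copying a mesh pose, head rotation relative to spine_05 drives additive vine-base offsets.
# Spring-based AnimDynamics move six head flowers, then RigidBody simulates the flaps.
# A backbling check zeroes backflapAlpha, swapping flap simulation for inverse-mapped, weighted constraints to spine, clavicle and pelvis bones.
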AnimBlueprint /BRCosmetics/Characters/Player/Female/Medium/Bodies/F_MED_Flower_Skeleton/Meshes/Parts/F_MED_Flower_Skeleton_FaceAcc_AnimBP
target_skeleton=&/F_MED_Flower_Skeleton_FaceAcc_Skeleton
graphs [
 AnimGraph {nodes [
   Anim Dynamics - Bone: dyn_head_flower_a(Alpha=1, ExternalForce="0,0,0") {
    angularDampingOverride=1
    angularSpringConstant=1500
    bAngularSpring=true
    bGravityOverrideInSimSpace=true
    bOverrideAngularDamping=true
    gravityScale=0.35
    simulationSpace=RootRelative
    alphaBoolBlend {blendOption=Linear}
    boundBone {boneName=dyn_head_flower_a}
    physicsBodyDefinitions [{
      sphereCollisionRadius=0
      boundBone {boneName=dyn_head_flower_a}
      boxExtents=<10 20 20>
      constraintSetup {angularLimitsMax=<0 0 0>, angularLimitsMin=<0 -5 0>, angularTarget=<0 0 0>}
      localJointOffset=<10 0 0>
     }]
   }
   Anim Dynamics - Bone: dyn_backflapA_a(Alpha=1, ExternalForce="0,0,0") {
    angularDampingOverride=1
    angularSpringConstant=1500
    bAngularSpring=true
    bGravityOverrideInSimSpace=true
    bOverrideAngularDamping=true
    gravityScale=0.35
    simulationSpace=RootRelative
    alphaBoolBlend {blendOption=Linear}
    boundBone {boneName=dyn_backflapA_a}
    physicsBodyDefinitions [{
      sphereCollisionRadius=0
      boundBone {boneName=dyn_backflapA_a}
      boxExtents=<10 30 30>
      constraintSetup {angularLimitsMax=<0 0 0>, angularLimitsMin=<0 0 -30>, angularTarget=<0 0 0>}
      localJointOffset=<15 0 0>
     }]
   }
   n1 = "Constraint - dyn_backflapA_b (PPPPPPP)"!(
    ComponentPose
    Alpha=1
    ConstraintWeights_0=0.0109
    ConstraintWeights_1:float=0.1004
    ConstraintWeights_2:float=0.238
    ConstraintWeights_3:float=0.2712
    ConstraintWeights_4:float=0.3095
    ConstraintWeights_5:float=0.0496
    ConstraintWeights_6:float=0.0204
   ) -> (Pose) {
    alphaBoolBlend {blendOption=Linear}
    alphaScaleBiasClamp {bMapRange=true}
    boneToModify {boneName=dyn_backflapA_b}
    constraintSetup [
     {targetBone {boneName=clavicle_l}}
     {targetBone {boneName=spine_01}}
     {targetBone {boneName=spine_02}}
     {targetBone {boneName=spine_03}}
     {targetBone {boneName=spine_04}}
     {targetBone {boneName=spine_05}}
     {targetBone {boneName=clavicle_r}}
    ]
    constraintWeights [1,1,1,1,1,1,1]
   }
   "Constraint - dyn_backflapB_b (PPPP)"!(
    ComponentPose=@n1
    Alpha=1
    ConstraintWeights_0=0.2748
    ConstraintWeights_1:float=0.3792
    ConstraintWeights_2:float=0.1146
    ConstraintWeights_3:float=0.2315
   ) -> (Pose) {
    alphaBoolBlend {blendOption=Linear}
    alphaScaleBiasClamp {bMapRange=true}
    boneToModify {boneName=dyn_backflapB_b}
    constraintSetup [{targetBone {boneName=spine_01}}, {targetBone {boneName=spine_02}}, {targetBone {boneName=spine_03}}, {targetBone {boneName=pelvis}}]
    constraintWeights [1,1,1,1]
   }
   RigidBody() {
    bClampLinearTranslationLimitToRefPose=true
    bOverrideWorldGravity=true
    overridePhysicsAsset=&/F_MED_Flower_Skeleton_FaceAcc_Physics
    simulationSpace=BaseBoneSpace
    alphaBoolBlend {blendOption=Linear}
    baseBoneRef {boneName=neck_02}
    componentLinearAccScale=<0.35 0.3 0.3>
    componentLinearVelScale=<0.35 0.3 0.3>
    overrideWorldGravity {z=-980}
   }
   n2 = Get MeshToCopy() -> (MeshToCopy:Skeletal Mesh Component Object Reference)
   n3 = Get headFwdBkRot() -> (headFwdBkRot:double)
   n4 = "Transform (Modify) Bone - Bone: dyn_vine_main_base"(
    ComponentPose=Local To Component(LocalPose=Copy Pose From Mesh(SourceMeshComponent=@n2))
    Alpha=@n3
    Translation="0,0,0"
    Rotation="0,0,-20"
    Scale="1,1,1"
   ) {
    rotationMode=BMM_Additive
    rotationSpace=BCS_BoneSpace
    translationMode=BMM_Additive
    translationSpace=BCS_BoneSpace
    alphaBoolBlend {blendOption=Linear}
    alphaScaleBiasClamp {bMapRange=true, inRange {max=20, min=-10}}
    boneToModify {boneName=dyn_vine_main_base}
    rotation {roll=-20}
   }
   n5 = "Transform (Modify) Bone - Bone: dyn_vine_main_base"(ComponentPose=@n4, Alpha=@n3, Translation="0,3.5,-2", Rotation="0,0,40", Scale="1,1,1") {
    rotationMode=BMM_Additive
    rotationSpace=BCS_BoneSpace
    scaleMode=BMM_Additive
    scaleSpace=BCS_BoneSpace
    translationMode=BMM_Additive
    translationSpace=BCS_BoneSpace
    alphaBoolBlend {blendOption=Linear}
    alphaScaleBiasClamp {bMapRange=true, inRange {max=-10, min=-50}, outRange {max=0, min=1}}
    boneToModify {boneName=dyn_vine_main_base}
    rotation {roll=40}
    translation {y=3.5, z=-2}
   }
   n6 = Get headSideToSide() -> (headSideToSide:double)
   n7 = "Transform (Modify) Bone - Bone: dyn_vine_main_base"(ComponentPose=@n5, Alpha=@n6, Translation="0,0,0", Rotation="0,30,0", Scale="1,1,1") {
    rotationMode=BMM_Additive
    rotationSpace=BCS_BoneSpace
    scaleMode=BMM_Additive
    scaleSpace=BCS_BoneSpace
    translationMode=BMM_Additive
    translationSpace=BCS_BoneSpace
    alphaBoolBlend {blendOption=Linear}
    alphaScaleBiasClamp {bMapRange=true, inRange {max=45}}
    boneToModify {boneName=dyn_vine_main_base}
    rotation {yaw=30}
   }
   n8 = "Transform (Modify) Bone - Bone: dyn_vine_main_base"(ComponentPose=@n7, Alpha=@n6, Translation="0,0,0", Rotation="0,-30,0", Scale="1,1,1") {
    rotationMode=BMM_Additive
    rotationSpace=BCS_BoneSpace
    scaleMode=BMM_Additive
    scaleSpace=BCS_BoneSpace
    translationMode=BMM_Additive
    translationSpace=BCS_BoneSpace
    alphaBoolBlend {blendOption=Linear}
    alphaScaleBiasClamp {bMapRange=true, inRange {max=0, min=-45}, outRange {max=0, min=1}}
    boneToModify {boneName=dyn_vine_main_base}
    rotation {yaw=-30}
   }
   n9 = "Constraint - dyn_backflapA_base (PPPPPPP)"!(
    ComponentPose=@n8
    Alpha=0
    ConstraintWeights_0=0.0869
    ConstraintWeights_1:float=0.0171
    ConstraintWeights_2:float=0.0512
    ConstraintWeights_3:float=0.0763
    ConstraintWeights_4:float=0.4178
    ConstraintWeights_5:float=0.2621
    ConstraintWeights_6:float=0.0885
   ) -> (Pose) {
    alphaBoolBlend {blendOption=Linear}
    boneToModify {boneName=dyn_backflapA_base}
    constraintSetup [
     {targetBone {boneName=clavicle_l}}
     {targetBone {boneName=spine_01}}
     {targetBone {boneName=spine_02}}
     {targetBone {boneName=spine_03}}
     {targetBone {boneName=spine_04}}
     {targetBone {boneName=spine_05}}
     {targetBone {boneName=clavicle_r}}
    ]
    constraintWeights [1,1,1,1,1,1,1]
   }
   n10 = "Constraint - dyn_backflapB_base (PPPPPPP)"!(
    ComponentPose=@n9
    Alpha=1
    ConstraintWeights_0=0.016
    ConstraintWeights_1:float=0.1029
    ConstraintWeights_2:float=0.2435
    ConstraintWeights_3:float=0.2753
    ConstraintWeights_4:float=0.3022
    ConstraintWeights_5:float=0.0472
    ConstraintWeights_6:float=0.0129
   ) -> (Pose) {
    alphaBoolBlend {blendOption=Linear}
    boneToModify {boneName=dyn_backflapB_base}
    constraintSetup [
     {targetBone {boneName=clavicle_l}}
     {targetBone {boneName=spine_01}}
     {targetBone {boneName=spine_02}}
     {targetBone {boneName=spine_03}}
     {targetBone {boneName=spine_04}}
     {targetBone {boneName=spine_05}}
     {targetBone {boneName=clavicle_r}}
    ]
    constraintWeights [1,1,1,1,1,1,1]
   }
   n11 = Anim Dynamics - Bone: dyn_head_flower_a!(ComponentPose=@n10, Alpha=1, GravityScale:float=1, GravityOverride:Vector="0,0,-500", ExternalForce="0,0,0") -> (Pose) {
    angularDampingOverride=0.9999
    angularSpringConstant=1000
    bAngularSpring=true
    bGravityOverrideInSimSpace=true
    bOverrideAngularDamping=true
    simulationSpace=RootRelative
    alphaBoolBlend {blendOption=Linear}
    boundBone {boneName=dyn_head_flower_a}
    componentAppliedLinearAccClamp=<1000 1000 1000>
    componentLinearAccScale=<0.1 0.1 0.1>
    componentLinearVelScale=<0.1 0.1 0.1>
    gravityOverride {z=-980}
    physicsBodyDefinitions [{
      sphereCollisionRadius=0
      boundBone {boneName=dyn_head_flower_a}
      boxExtents=<12 10 10>
      constraintSetup {angularLimitsMax=<0 2 8>, angularLimitsMin=<0 -12 -12>, angularTarget=<0 0 0>}
      localJointOffset=<0 1.2 0.3>
     }]
    relativeSpaceBone {boneName=head}
   }
   n12 = Anim Dynamics - Bone: dyn_head_flower_b!(ComponentPose=@n11, Alpha=1, GravityScale:float=1, GravityOverride:Vector="0,0,-500", ExternalForce="0,0,0") -> (Pose) {
    angularDampingOverride=0.9999
    angularSpringConstant=1000
    bAngularSpring=true
    bGravityOverrideInSimSpace=true
    bOverrideAngularDamping=true
    simulationSpace=RootRelative
    alphaBoolBlend {blendOption=Linear}
    boundBone {boneName=dyn_head_flower_b}
    componentAppliedLinearAccClamp=<1000 1000 1000>
    componentLinearAccScale=<0.2 0.2 0.2>
    componentLinearVelScale=<0.2 0.2 0.2>
    gravityOverride {z=-980}
    physicsBodyDefinitions [{
      sphereCollisionRadius=0
      boundBone {boneName=dyn_head_flower_b}
      boxExtents=<12 10 10>
      constraintSetup {angularLimitsMax=<0 8 10>, angularLimitsMin=<0 -12 -8>, angularTarget=<0 0 0>}
      localJointOffset=<0.6 1.3 -0.3>
     }]
    relativeSpaceBone {boneName=head}
   }
   n13 = Anim Dynamics - Bone: dyn_head_flower_c!(ComponentPose=@n12, Alpha=1, GravityScale:float=1, GravityOverride:Vector="0,0,-500", ExternalForce="0,0,0") -> (Pose) {
    angularDampingOverride=0.9999
    angularSpringConstant=1000
    bAngularSpring=true
    bGravityOverrideInSimSpace=true
    bOverrideAngularDamping=true
    simulationSpace=RootRelative
    alphaBoolBlend {blendOption=Linear}
    boundBone {boneName=dyn_head_flower_c}
    componentAppliedLinearAccClamp=<1000 1000 1000>
    componentLinearAccScale=<0.1 0.1 0.1>
    componentLinearVelScale=<0.1 0.1 0.1>
    gravityOverride {z=-980}
    physicsBodyDefinitions [{
      sphereCollisionRadius=0
      boundBone {boneName=dyn_head_flower_c}
      boxExtents=<12 10 10>
      constraintSetup {angularLimitsMax=<0 12 12>, angularLimitsMin=<0 -12 -12>, angularTarget=<0 0 0>}
      localJointOffset=<1 0 0.2>
     }]
    relativeSpaceBone {boneName=head}
   }
   n14 = Anim Dynamics - Bone: dyn_head_flower_d!(ComponentPose=@n13, Alpha=1, GravityScale:float=1, GravityOverride:Vector="0,0,-500", ExternalForce="0,0,0") -> (Pose) {
    angularDampingOverride=0.9999
    angularSpringConstant=1000
    bAngularSpring=true
    bGravityOverrideInSimSpace=true
    bOverrideAngularDamping=true
    simulationSpace=RootRelative
    alphaBoolBlend {blendOption=Linear}
    boundBone {boneName=dyn_head_flower_d}
    componentAppliedLinearAccClamp=<1000 1000 1000>
    componentLinearAccScale=<0.1 0.1 0.1>
    componentLinearVelScale=<0.1 0.1 0.1>
    gravityOverride {z=-980}
    physicsBodyDefinitions [{
      sphereCollisionRadius=0
      boundBone {boneName=dyn_head_flower_d}
      boxExtents=<12 10 10>
      constraintSetup {angularLimitsMax=<0 12 12>, angularLimitsMin=<0 -12 -12>, angularTarget=<0 0 0>}
      localJointOffset=<1 -0.6 0.3>
     }]
    relativeSpaceBone {boneName=head}
   }
   n15 = Anim Dynamics - Bone: dyn_head_flower_e!(ComponentPose=@n14, Alpha=1, GravityScale:float=1, GravityOverride:Vector="0,0,-500", ExternalForce="0,0,0") -> (Pose) {
    angularDampingOverride=0.9999
    angularSpringConstant=1000
    bAngularSpring=true
    bGravityOverrideInSimSpace=true
    bOverrideAngularDamping=true
    simulationSpace=RootRelative
    alphaBoolBlend {blendOption=Linear}
    boundBone {boneName=dyn_head_flower_e}
    componentAppliedLinearAccClamp=<1000 1000 1000>
    componentLinearAccScale=<0.1 0.1 0.1>
    componentLinearVelScale=<0.1 0.1 0.1>
    gravityOverride {z=-980}
    physicsBodyDefinitions [{
      sphereCollisionRadius=0
      boundBone {boneName=dyn_head_flower_e}
      boxExtents=<12 10 10>
      constraintSetup {angularLimitsMax=<0 8 8>, angularLimitsMin=<0 -8 -8>, angularTarget=<0 0 0>}
      localJointOffset=<0.5 -1.2 0>
     }]
    relativeSpaceBone {boneName=head}
   }
   n16 = Anim Dynamics - Bone: dyn_head_flower_f!(ComponentPose=@n15, Alpha=1, GravityScale:float=1, GravityOverride:Vector="0,0,-500", ExternalForce="0,0,0") -> (Pose) {
    angularDampingOverride=0.999
    angularSpringConstant=1000
    bAngularSpring=true
    bGravityOverrideInSimSpace=true
    bOverrideAngularDamping=true
    simulationSpace=RootRelative
    alphaBoolBlend {blendOption=Linear}
    boundBone {boneName=dyn_head_flower_f}
    componentAppliedLinearAccClamp=<1000 1000 1000>
    componentLinearAccScale=<0.1 0.1 0.1>
    componentLinearVelScale=<0.1 0.1 0.1>
    gravityOverride {z=-980}
    physicsBodyDefinitions [{
      sphereCollisionRadius=0
      boundBone {boneName=dyn_head_flower_f}
      boxExtents=<12 10 10>
      constraintSetup {angularLimitsMax=<0 1 5>, angularLimitsMin=<0 -8 -5>, angularTarget=<0 0 0>}
      localJointOffset=<0 -0.6 0.1>
     }]
    relativeSpaceBone {boneName=head}
   }
   n17 = Get backflapAlpha() -> (backflapAlpha:double)
   n18 = RigidBody(ComponentPose=@n16, Alpha=@n17) {
    bClampLinearTranslationLimitToRefPose=true
    bOverrideWorldGravity=true
    overridePhysicsAsset=&/F_MED_Flower_Skeleton_FaceAcc_Flap_Dyn_Physics
    simulationSpace=BaseBoneSpace
    alphaBoolBlend {blendOption=Linear}
    baseBoneRef {boneName=pelvis}
    componentLinearAccScale=<0.45 0.4 0.4>
    componentLinearVelScale=<0.45 0.4 0.4>
    overrideWorldGravity {z=-980}
   }
   n19 = "Transform (Modify) Bone - Bone: dyn_backflapA_base"(ComponentPose=@n18, Alpha=@n17, Translation="0,-1,0", Rotation="0,0,0", Scale="1,1,1") {
    rotationMode=BMM_Additive
    rotationSpace=BCS_BoneSpace
    translationMode=BMM_Additive
    translationSpace=BCS_BoneSpace
    alphaBoolBlend {blendOption=Linear}
    alphaScaleBiasClamp {inRange {max=20}}
    boneToModify {boneName=dyn_backflapA_base}
    translation {y=-1}
   }
   n20 = "Transform (Modify) Bone - Bone: dyn_backflapB_base"(ComponentPose=@n19, Alpha=@n17, Translation="0,-1,0", Rotation="0,0,0", Scale="1,1,1") {
    rotationMode=BMM_Additive
    rotationSpace=BCS_BoneSpace
    translationMode=BMM_Additive
    translationSpace=BCS_BoneSpace
    alphaBoolBlend {blendOption=Linear}
    alphaScaleBiasClamp {inRange {max=20}}
    boneToModify {boneName=dyn_backflapB_base}
    translation {y=-1}
   }
   n21 = Get backflapAlpha()
   n22 = "Constraint - dyn_backflapA_b (PPPPPPP)"(
    ComponentPose=@n20
    Alpha=@n21
    ConstraintWeights_0=0.0654
    ConstraintWeights_1=0.0289
    ConstraintWeights_2=0.0774
    ConstraintWeights_3=0.1129
    ConstraintWeights_4=0.4607
    ConstraintWeights_5=0.1889
    ConstraintWeights_6=0.0658
   ) {
    alphaBoolBlend {blendOption=Linear}
    alphaScaleBiasClamp {bMapRange=true, outRange {max=0, min=1}}
    boneToModify {boneName=dyn_backflapA_b}
    constraintSetup [
     {targetBone {boneName=clavicle_l}}
     {targetBone {boneName=spine_01}}
     {targetBone {boneName=spine_02}}
     {targetBone {boneName=spine_03}}
     {targetBone {boneName=spine_04}}
     {targetBone {boneName=spine_05}}
     {targetBone {boneName=clavicle_r}}
    ]
    constraintWeights [1,1,1,1,1,1,1]
   }
   n23 = "Constraint - dyn_backflapA_c (PPPPPPP)"!(
    ComponentPose=@n22
    Alpha=@n21
    ConstraintWeights_0=0.0213
    ConstraintWeights_1:float=0.0862
    ConstraintWeights_2:float=0.2119
    ConstraintWeights_3:float=0.2621
    ConstraintWeights_4:float=0.3377
    ConstraintWeights_5:float=0.0594
    ConstraintWeights_6:float=0.0214
   ) -> (Pose) {
    alphaBoolBlend {blendOption=Linear}
    alphaScaleBiasClamp {bMapRange=true, outRange {max=0, min=1}}
    boneToModify {boneName=dyn_backflapA_c}
    constraintSetup [
     {targetBone {boneName=clavicle_l}}
     {targetBone {boneName=spine_01}}
     {targetBone {boneName=spine_02}}
     {targetBone {boneName=spine_03}}
     {targetBone {boneName=spine_04}}
     {targetBone {boneName=spine_05}}
     {targetBone {boneName=clavicle_r}}
    ]
    constraintWeights [1,1,1,1,1,1,1]
   }
   n24 = "Constraint - dyn_backflapB_a (PPPPPPP)"!(
    ComponentPose=@n23
    Alpha=@n21
    ConstraintWeights_0=0.0654
    ConstraintWeights_1:float=0.0289
    ConstraintWeights_2:float=0.0774
    ConstraintWeights_3:float=0.1129
    ConstraintWeights_4:float=0.4607
    ConstraintWeights_5:float=0.1889
    ConstraintWeights_6:float=0.0658
   ) -> (Pose) {
    alphaBoolBlend {blendOption=Linear}
    alphaScaleBiasClamp {bMapRange=true, outRange {max=0, min=1}}
    boneToModify {boneName=dyn_backflapB_a}
    constraintSetup [
     {targetBone {boneName=clavicle_l}}
     {targetBone {boneName=spine_01}}
     {targetBone {boneName=spine_02}}
     {targetBone {boneName=spine_03}}
     {targetBone {boneName=spine_04}}
     {targetBone {boneName=spine_05}}
     {targetBone {boneName=clavicle_r}}
    ]
    constraintWeights [1,1,1,1,1,1,1]
   }
   n25 = "Constraint - dyn_backflapB_b (PPPPPPP)"!(
    ComponentPose=@n24
    Alpha=@n21
    ConstraintWeights_0=0.0213
    ConstraintWeights_1:float=0.0862
    ConstraintWeights_2:float=0.2119
    ConstraintWeights_3:float=0.2621
    ConstraintWeights_4:float=0.3377
    ConstraintWeights_5:float=0.0594
    ConstraintWeights_6:float=0.0214
   ) -> (Pose) {
    alphaBoolBlend {blendOption=Linear}
    alphaScaleBiasClamp {bMapRange=true, outRange {max=0, min=1}}
    boneToModify {boneName=dyn_backflapB_b}
    constraintSetup [
     {targetBone {boneName=clavicle_l}}
     {targetBone {boneName=spine_01}}
     {targetBone {boneName=spine_02}}
     {targetBone {boneName=spine_03}}
     {targetBone {boneName=spine_04}}
     {targetBone {boneName=spine_05}}
     {targetBone {boneName=clavicle_r}}
    ]
    constraintWeights [1,1,1,1,1,1,1]
   }
   n26 = "Constraint - dyn_backflapB_c (PPPP)"!(
    ComponentPose=@n25
    Alpha=@n21
    ConstraintWeights_0=0.2513
    ConstraintWeights_1:float=0.4111
    ConstraintWeights_2:float=0.1642
    ConstraintWeights_3:float=0.1735
   ) -> (Pose) {
    alphaBoolBlend {blendOption=Linear}
    alphaScaleBiasClamp {bMapRange=true, outRange {max=0, min=1}}
    boneToModify {boneName=dyn_backflapB_c}
    constraintSetup [{targetBone {boneName=spine_01}}, {targetBone {boneName=spine_02}}, {targetBone {boneName=spine_03}}, {targetBone {boneName=pelvis}}]
    constraintWeights [1,1,1,1]
   }
   Output Pose(Result=Component To Local(ComponentPose=@n26)) {layerGroup=None}
  ]}
 EventGraph {nodes [
   Event BlueprintUpdateAnimation()
   > n1 = Sequence!(execute) -> (then_0, then_1, then_2:Exec, then_3:Exec)
   n2 = GetOwnerSkeletalMeshComponent()
   n3 = BreakRotator(InRot=BreakTransform(InTransform=GetDeltaTransformFromRefPose(self=@n2, BoneName=head, BaseName=spine_05)).Rotation)
   Set headFwdBkRot(execute=@n1.then_1, headFwdBkRot:double=@n3.Yaw) -> (Output_Get:double)
   > Set headSideToSide(headSideToSide:double=@n3) -> (Output_Get:double)
   n4 = GetDeltaTransformFromRefPose(self=@n2, BoneName=spine_05, BaseName=spine_03)
   Set torsoFwdBkRot(execute=@n1.then_2, torsoFwdBkRot:double=BreakRotator(InRot=BreakTransform(InTransform=@n4).Rotation).Yaw) -> (Output_Get:double)
   n5 = Get playerPawn() -> (playerPawn:Fort Player Pawn Athena Object Reference)
   Is Valid(exec=@n1.then_0, InputObject=@n5)
   n6 = CheckForBackbling()
   > n7 = Set IsBackpackOn(IsBackpackOn:bool=OR Boolean(A=@n6, B=@n6.bIsBackblingLong)) -> (Output_Get:bool)
   > n8 = Branch(Condition=@n7)
   > Set backflapAlpha(backflapAlpha:double=0) -> (Output_Get:double)
   Set backflapAlpha(execute=@n8.else, backflapAlpha=1)
   n9 = Get CurrentLOD()
   n10 = Get bRigidBodyEnabled()
   Set isRigidBodyDisabled(
    execute=@n1.then_3
    isRigidBodyDisabled:bool=OR Boolean(A=NOT Boolean(A=@n10), B="integer > integer"(A=@n9, B=1) {is=title K2Node_CallFunction})
   ) -> (Output_Get:bool)
   Event BlueprintInitializeAnimation()
   > Set MeshToCopy(MeshToCopy:Skeletal Mesh Component Object Reference=GetOwnerSkeletalMeshComponent()) -> (Output_Get:Skeletal Mesh Component Object Reference)
   n11 = TryGetPawnOwner()
   > Is Valid(InputObject=@n11)
   > n12 = Cast To FortPlayerPawnAthena(Object=@n11)
   > Set playerPawn(playerPawn:Fort Player Pawn Athena Object Reference=@n12) -> (Output_Get:Fort Player Pawn Athena Object Reference)
   TryGetPawnOwner()
   "Not Equal (Object)"!(A:Object Reference=GetSkeletalMeshForPartType(self=@n5, PartType=Backpack), B:Object Reference) -> (ReturnValue:bool)
  ]}
]
variables [
 backflapAlpha {type=number}
 headFwdBkRot {type=number}
 headSideToSide {type=number}
 IsBackpackOn {type=boolean}
 isRigidBodyDisabled {type=boolean}
 MeshToCopy {type=SkeletalMeshComponent}
 playerPawn {type=FortPlayerPawnAthena}
 torsoFwdBkRot {type=number}
]
